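# Niagara system for the Deep Space drift trail vehicle cosmetic (MP variant), attaching glowing effects to the car's wheels.
# Burst-spawned, looping particles pinned to skeletal mesh sockets draw wheel-wrap meshes plus offset mesh and sprite accents.
# Two quality-scaled ribbons trail behind, inheriting vehicle movement that fades with age while width and color follow curves.
# All emitters mask per tire via FN_DisableTireFX and take painted-variant palettes, with bIsPainted adjusting intensity and material values.
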NiagaraSystem /VehicleCosmetics/DriftTrail/DriftTrail_DeepSpace/FX/NS_DrifTrail_DeepSpace_MP
dependencies {
 usedDataInterfaces [
  &NiagaraDataInterfaceArrayFloat3
  &NiagaraDataInterfaceColorCurve
  &NiagaraDataInterfaceCurve
  &NiagaraDataInterfaceSkeletalMesh
 ]
 usedDynamicInputs [&MergeHelper_FloatSwitch]
 usedModules [
  &self:SystemSpawnScript.SystemScriptSource.SystemScriptGraph.NiagaraNodeAssignment_0.SetVariables_#1
  &FN_UnpackVehicleFlags
  /VehicleCosmetics/FX/Niagara/Modules/VC_SetVehicleSpeedAttribs
  &SystemState
  /Game/Effects/Niagara/Modules/Vehicle/FN_UpdateVehicleAttribs
  /Game/Effects/Niagara/Modules/Vehicle/FN_VehicleFlags_Cheatcodes
 ]
}
emitters {
 NE_VC_DriftTrail_Sprite_Parent {
  emitterSpawnScript [/SetVariables(
    Emitter.ColorP:NiagaraFloat=&MakeCustomFloatFromBool(Bool=$User.bIsPainted, True Float=0, False Float=1)
    Emitter.Glass_Colo:NiagaraFloat=&MakeCustomFloatFromBool(Bool=$User.bIsPainted, True Float=1, False Float=0)
   ) {isSetParameters=true}]
  emitterUpdateScript [
   &EmitterState(
    Inactive Response=::NewEnumerator1
    Life Cycle Mode=::NewEnumerator1
    Loop Behavior=::NewEnumerator1
    Loop Duration Mode=::NewEnumerator1
    UseLoopDelay=false
   )
   &SpawnRate(
    SpawnRate=/VehicleCosmetics/FX/Niagara/Modules/VC_Mask_Floats(
     FrontEnd Value:NiagaraFloat=200
     blsFrontEndPreview:NiagaraBool=false
     SpeedMin Value:NiagaraFloat=120
     SpeedMax Value:NiagaraFloat=180
     Speed Alpha:NiagaraFloat=$System.Speed_Alpha
    )
    Use Spawn Probability=false
   ) {enabled=false}
   &SpawnBurst_Instantaneous(Spawn Count=4, Use Spawn Probability=false, UseLoopCountLimit=false)
   &SpawnBurst_Instantaneous(Spawn Count=4, Spawn Group=1, Use Spawn Probability=false, UseLoopCountLimit=false) {module=SpawnBurst_Instantaneous001}
  ]
  particleSpawnScript [
   /SetVariables(Particles.VisibilityTag=$Emitter.SpawnGroup) {isSetParameters=true}
   &VC_PaintedVariantsOverride(
    Black=<rgba 0.1146 0.1146 0.1146 2>
    BurntSienna=<rgba 0.5156 0.1324 0.0295 2>
    Cobalt=<rgba 0.3021 0.3602 1 2>
    Crimson=<rgba 1 0.2708 0.2906 1>
    ForestGreen=<rgba 0.1979 1 0.2384 1>
    Grey=<rgba 0.2188 0.2188 0.2188 2>
    Lime=<rgba 0.3824 0.6615 0.1034 1>
    Orange=<rgba 0.6302 0.2284 0.0131 1>
    Pink=<rgba 1 0.3177 0.6589 1>
    Purple=<rgba 0.4263 0.3385 1 1>
    R1 Black:LinearColor=<rgba 0.0417 0.0417 0.0417 1>
    R1 BurntSienna:LinearColor=<rgba 0.0642 0.0198 0.0079 1>
    R1 DefaultColor=<rgba 0.9578 0.3073 1 1>
    R1 ForestGreen:LinearColor=<rgba 0 0.5 0.0253 1>
    R1 Orange:LinearColor=<rgba 0.4271 0.149 0 1>
    R1 Pink:LinearColor=<rgba 0.3021 0.0739 0.2191 1>
    R1 Purple:LinearColor=<rgba 0.0718 0.0438 0.4427 1>
    Saffron=<rgba 0.7812 0.6284 0.1261 1>
    SkyBlue=<rgba 0.2088 0.5828 0.7708 1>
    TitaniumWhite=<rgba 0.9271 0.9271 0.9271 1>
    Use R1 Painted Overrides=true
   )
   &InitializeParticle(
    Color Mode=::NewEnumerator0
    Lifetime=0.1
    Mesh Scale=&MergeHelper_VectorSwitch(
     0=<0.3 0.2 0.4>
     1=&VectorFromFloat(Value=1.5)
     2=?
     3=?
     Index=$Emitter.SpawnGroup
     Num=::NewEnumerator0
    )
    Mesh Scale Mode=::NewEnumerator4
    Sprite Size Mode=::NewEnumerator3
    Uniform Sprite Size=130
    UsePositionOffset=false
    Write Material Random=true
   )
   &PartitionParticles()
   &SkeletalMeshLocation(
    Bone / Socket Index=&Add_Integer(A=$Particles.PartitionParticles.Partition, B=4)
    Bone / Socket Sampling Mode=::NewEnumerator5
    Interpolate Mesh Sampling=false
    MaskSpawnGroup=false
    Mesh Sampling Type=::NewEnumerator6
    Sampled Position Offset=&MergeHelper_VectorSwitch(0=<-80 0 -30>, 1=<-75 0 -40>, 2=<-70 0 -30>, 3=?, Index=&SpawnGroup(), Num=::NewEnumerator1)
    Skeletal Mesh=$System.SkeletalMesh
   )
   &AddVelocity(Velocity=/VehicleCosmetics/FX/Niagara/Modules/MergeEmitters/VC_VectorBySpawnGroup(
     SpawnGroup or ExecIndex:NiagaraInt32=0
     CustomInt:NiagaraInt32=0
     MaximumSpawnGroups:NiagaraInt32=2
     Vector Group 0:Vector3f=<-1000 0 0>
     Vector Group 1:Vector3f=<0 0 0>
     Vector Group 2:Vector3f=<0 0 0>
     Vector Group 3:Vector3f=<0 0 0>
     Vector Group 4:Vector3f=<0 0 0>
     Vector Group 5:Vector3f=<0 0 0>
    )) {enabled=false}
  ]
  particleUpdateScript [
   &FN_DisableTireFX(
    Mask Index Attribute:NiagaraInt32=$Particles.PartitionParticles.Partition
    Mask Index FL=2
    Mask Index FR=3
    Mask Index RL=0
    Mask Index RR=1
   )
   &ParticleState(Kill Particles When Lifetime Has Elapsed=false, Loop Particles Lifetime=true)
   &UpdateMeshOrientation(Rotation Vector=&MergeHelper_VectorSwitch(0=<0 0 0>, 1=<0.03 0.03 1>, 2=?, 3=?, Index=&SpawnGroup(), Num=::NewEnumerator0))
   &SolveForcesAndVelocity(Clamp Velocity=false, Limit Acceleration=false)
   &DynamicMaterialParameters(
    Index 0 Param 1=&MakeCustomFloatFromBool(Bool=$User.bIsPainted, True Float=0.5, False Float=0)
    Index 0 Param 2=1
    Module.Index0Float0Write=true
    Module.Index0Float1Write=true
    Module.Index0Float2Write=false
    Module.Index0Float3Write=false
   )
   /SetVariables(Particles.MeshColor1:LinearColor=&MultiplyLinearColor(
     Linear Color=$Particles.Initial.R1 Color
     Scale Factor=&MakeVector4FromVectorAndFloat(
      "Vector (RGB)"=&VectorFromFloat(Value=&MakeCustomFloatFromBool(Bool=$User.bIsPainted, True Float=5, False Float=100))
      "Float (Alpha)"=1
     )
    )) {isSetParameters=true}
  ]
  rendererClasses [&NiagaraMeshRendererProperties, &NiagaraSpriteRendererProperties]
  renderers [
   {rendererClass=&NiagaraMeshRendererProperties, rendererIndex=0}
   {rendererClass=&NiagaraMeshRendererProperties, rendererIndex=1}
   {rendererClass=&NiagaraSpriteRendererProperties, rendererIndex=2}
  ]
 }
 NE_VC_DriftTrail_WheelWrap_Parent001 {
  emitterSpawnScript [/SetVariables(
    Emitter.DesaturationContrast:NiagaraFloat=&SelectFloatFromArray(
     Direct Array Index=$System.ColorVariantID
     Float Selection Array=DataInterface() {propertyValues {FloatData [1,1,1.2,1,0.7,0.7,0.9,1,0.8,0.7,0.7,0.7,0.7,0.7,0.9]}}
    )
    Emitter.DiffuseColor:NiagaraFloat=&MakeCustomFloatFromBool(Bool=$User.bIsPainted, True Float=0, False Float=1)
   ) {isSetParameters=true}]
  emitterUpdateScript [
   &EmitterState(
    Inactive Response=::NewEnumerator1
    Life Cycle Mode=::NewEnumerator1
    Loop Behavior=::NewEnumerator1
    Loop Duration Mode=::NewEnumerator1
    UseLoopDelay=false
   )
   &SpawnBurst_Instantaneous(Spawn Count=2, Use Spawn Probability=false, UseLoopCountLimit=false)
  ]
  particleSpawnScript [
   &VC_PaintedVariantsOverride(
    Black=<rgba 0.151 0.151 0.151 0.8>
    BurntSienna=<rgba 1 0.4005 0.2396 0.6>
    Cobalt=<rgba 0.6146 0.6839 2 0.6>
    Crimson=<rgba 1.4792 0.4468 0.4748 0.6>
    ForestGreen=<rgba 0.623 2 0.4215 0.6>
    Grey=<rgba 0.7188 0.7188 0.7188 0.6>
    Lime=<rgba 1.2812 2 0.5625 0.6>
    Orange=<rgba 1.4 0.5221 0.1458 0.7>
    Pink=<rgba 2 0.7188 1.3594 0.75>
    Purple=<rgba 0.7798 0.4 2 0.6>
    R1 Black:LinearColor=<rgba 0.276 0.276 0.276 1>
    R1 BurntSienna:LinearColor=<rgba 1 0.3693 0.2 1>
    R1 Cobalt:LinearColor=<rgba 0.2 0.29 2 1>
    R1 Crimson:LinearColor=<rgba 1 0.224 0.245 1>
    R1 DefaultColor=<rgba 1 0.5105 0.099 1>
    R1 ForestGreen:LinearColor=<rgba 0.4271 1 0.456 1>
    R1 Lime:LinearColor=<rgba 0.625 1 0.25 1>
    R1 Orange:LinearColor=<rgba 1 0.2875 0.05 1>
    R1 Pink:LinearColor=<rgba 1 0.4375 0.7955 1>
    R1 Purple:LinearColor=<rgba 0.487 0.3906 1 1>
    R1 Saffron:LinearColor=<rgba 1 0.8384 0.3073 1>
    R1 SkyBlue:LinearColor=<rgba 0.3698 0.7891 1 1>
    R1 TitaniumWhite:LinearColor=<rgba 1 1 1 1>
    Saffron=<rgba 1.5 1.3633 0.3203 0.6>
    SkyBlue=<rgba 0.625 1.5399 2 0.6>
    TitaniumWhite=<rgba 2 2 2 0.6>
    Use R1 Painted Overrides=true
   )
   &InitializeParticle(
    Color Mode=::NewEnumerator0
    Lifetime=0.1
    Mesh Scale Mode=::NewEnumerator4
    Sprite Rotation Mode=::NewEnumerator2
    Sprite Size Mode=::NewEnumerator3
    UsePositionOffset=false
    Write Material Random=true
   )
   &PartitionParticles()
   &SkeletalMeshLocation(
    Bone / Socket Index=&Add_Integer(A=$Particles.PartitionParticles.Partition, B=4)
    Bone / Socket Sampling Mode=::NewEnumerator5
    Interpolate Mesh Sampling=false
    MaskSpawnGroup=false
    Mesh Sampling Type=::NewEnumerator6
    Skeletal Mesh=$System.SkeletalMesh
   )
  ]
  particleUpdateScript [
   &FN_DisableTireFX(
    Mask Index Attribute:NiagaraInt32=$Particles.PartitionParticles.Partition
    Mask Index FL=2
    Mask Index FR=3
    Mask Index RL=0
    Mask Index RR=1
   )
   &VC_WheelWrapScale(Wheel Index Attribute:NiagaraInt32=&Add_Integer(A=$Particles.PartitionParticles.Partition, B=2))
   &ParticleState(Kill Particles When Lifetime Has Elapsed=false, Loop Particles Lifetime=true)
   &UpdateMeshOrientation(Rotation Coordinate Space=::NewEnumerator2, Rotation Vector=<0 1 0>) {enabled=false}
   &SkeletalMeshLocation(
    Bone / Socket Index=&Add_Integer(A=$Particles.PartitionParticles.Partition, B=4)
    Bone / Socket Sampling Mode=::NewEnumerator5
    MaskSpawnGroup=false
    Mesh Sampling Type=::NewEnumerator6
    Sampled Position Offset=&SelectVectorFromArray(
     Direct Array Index=$Particles.PartitionParticles.Partition
     Vector Selection Array=DataInterface() {propertyValues {FloatData [<0 0 0>, <0 1 0>]}}
    )
    Skeletal Mesh=$System.SkeletalMesh
   ) {module=SkeletalMeshLocation001}
   &SolveForcesAndVelocity(Clamp Velocity=false, Limit Acceleration=false)
   /SetVariables(Particles.WrapColor1:LinearColor=&MultiplyLinearColor(
     Linear Color=$Particles.Initial.R1 Color
     Scale Factor=&MakeVector4FromVectorAndFloat(
      "Vector (RGB)"=&MakeCustomVectorFromBool(
       Bool=$User.bIsPainted
       True Vector=&VectorFromFloat(Value=8)
       False Vector=&VectorFromFloat(Value=5)
      )
      "Float (Alpha)"=3
     )
    )) {isSetParameters=true}
  ]
  rendererClasses [&NiagaraMeshRendererProperties]
  renderers [
   {rendererClass=&NiagaraMeshRendererProperties, rendererIndex=0}
   {rendererClass=&NiagaraMeshRendererProperties, rendererIndex=1}
  ]
 }
 Ribbons002 {
  emitterSpawnScript [/SetVariables(
    Emitter.DesaturationContrast:NiagaraFloat=&SelectFloatFromArray(
     Direct Array Index=$System.ColorVariantID
     Float Selection Array=DataInterface() {propertyValues {FloatData [1,1,1.2,0.7,0.7,0.7,0.9,1,0.8,0.7,0.8,0.7,0.7,0.7,0.9]}}
    )
    Emitter.DiffuseMode:NiagaraFloat=&MakeCustomFloatFromBool(Bool=$User.bIsPainted, True Float=0, False Float=1)
   ) {isSetParameters=true}]
  emitterUpdateScript [
   &EmitterState(Inactive Response=::NewEnumerator1, Life Cycle Mode=::NewEnumerator1, Loop Duration=0.5, UseLoopDelay=false)
   &VC_RateQualityHandler(High Spawn Rate=60, Low Spawn Rate=30, Medium Spawn Rate=45, Use Custom Rate?=true)
   &VC_SpawnRibbon_v1(Number of Ribbons=2, SpawnRate=$Emitter.Spawn Rate)
  ]
  particleSpawnScript [
   &VC_PaintedVariantsOverride(Orange=<rgba 0.6724 0.2346 0 1>)
   &VC_InitializeRibbon_v1(
    Color=$Particles.Color
    Custom Offset=true
    Enable Skeletal Position=true
    Lifetime=0.7
    Local Vector Offset=&AddVectorToPosition(Position=<-60 0 -27>, Vector=$User.offset)
    Offset Amount=1
    Ribbon Width=40
   )
   &AddVelocity(Velocity=<1500 0 0>, Velocity Speed Scale=-1) {module=AddVelocity001}
  ]
  particleUpdateScript [
   &FN_DisableTireFX(Mask Index Attribute:NiagaraInt32=$Particles.SpawnGroup, Mask Index FL=2, Mask Index FR=3, Mask Index RL=0, Mask Index RR=1)
   &VC_PaintedVariantsOverride(
    Black=<rgba 0.151 0.151 0.151 0.8>
    BurntSienna=<rgba 1 0.4005 0.2396 0.6>
    Cobalt=<rgba 0.6146 0.6839 2 0.6>
    Crimson=<rgba 1.4792 0.4468 0.4748 0.6>
    ForestGreen=<rgba 0.623 2 0.4215 0.6>
    Grey=<rgba 0.7188 0.7188 0.7188 0.6>
    Lime=<rgba 1.2812 2 0.5625 0.6>
    Orange=<rgba 1.4 0.5221 0.1458 0.7>
    Pink=<rgba 2 0.7188 1.3594 0.75>
    Purple=<rgba 0.7798 0.4 2 0.6>
    R1 Black:LinearColor=<rgba 0.1667 0.1667 0.1667 1>
    R1 BurntSienna:LinearColor=<rgba 0.8854 0.2928 0.1337 1>
    R1 Cobalt:LinearColor=<rgba 0.7656 0.8773 3 0.5826>
    R1 Crimson:LinearColor=<rgba 3 1.0312 1.0846 0.287>
    R1 DefaultColor=&ColorFromCurve(ColorCurve=DataInterface() {propertyValues {
       AlphaCurve {keys [<at 0 1 RCIM_Cubic>,<at 1 0 RCIM_Cubic>]}
       BlueCurve {keys [<at 0 0>,<at 0.0505 0.8347 RCIM_Cubic>]}
       GreenCurve {keys [<at 0 0.2877>,<at 0.0505 0 RCIM_Cubic>]}
       RedCurve {keys [<at 0 1>,<at 0.0505 1 RCIM_Cubic>]}
      }})
    R1 ForestGreen:LinearColor=<rgba 0.1825 0.9635 0.1394 0.5>
    R1 Grey:LinearColor=<rgba 0.474 0.474 0.474 0.5652>
    R1 Lime:LinearColor=<rgba 0.4567 0.776 0.1374 0.5391>
    R1 Orange:LinearColor=<rgba 1 0.3488 0 0.6261>
    R1 Pink:LinearColor=<rgba 1 0.5312 0.8295 0.5913>
    R1 Purple:LinearColor=<rgba 0.5043 0.4375 1 0.7565>
    R1 Saffron:LinearColor=<rgba 0.7344 0.613 0.2142 0.5826>
    R1 SkyBlue:LinearColor=<rgba 0.2234 0.7089 0.9531 0.6696>
    R1 TitaniumWhite:LinearColor=<rgba 1 1 1 0.6>
    R2 Black:LinearColor=<rgba 0.1458 0.1458 0.1458 1>
    R2 BurntSienna:LinearColor=<rgba 0.65 0.24 0.13 1>
    R2 Cobalt:LinearColor=<rgba 0.4922 0.5108 1.5 0.3>
    R2 Crimson:LinearColor=<rgba 1.5625 0.2523 0.2878 0.5>
    R2 DefaultColor:LinearColor=&ColorFromCurve(ColorCurve=DataInterface() {propertyValues {
       AlphaCurve {keys [<at 0 1 RCIM_Cubic>,<at 1 0 RCIM_Cubic>]}
       BlueCurve {keys [<at 0.0131 0>,<at 0.8659 0.8347 RCIM_Cubic>]}
       GreenCurve {keys [<at 0.0131 0.4309>,<at 0.8659 0 RCIM_Cubic>]}
       RedCurve {keys [<at 0.0131 1>,<at 0.8659 1 RCIM_Cubic>]}
      }})
    R2 ForestGreen:LinearColor=<rgba 0.2865 1 0.3225 0.6>
    R2 Grey:LinearColor=<rgba 0.2396 0.2396 0.2396 0.5>
    R2 Lime:LinearColor=<rgba 0.5521 1 0.1042 0.5>
    R2 Orange:LinearColor=<rgba 2 0.8194 0.2292 0.3>
    R2 Pink:LinearColor=<rgba 1 0.4531 0.8011 0.5>
    R2 Purple:LinearColor=<rgba 0.3832 0.3229 1 0.8522>
    R2 Saffron:LinearColor=<rgba 1.2396 1.0061 0.2389 0.3>
    R2 SkyBlue:LinearColor=<rgba 0.1576 0.6627 0.9167 0.3>
    R2 TitaniumWhite:LinearColor=<rgba 1 1 1 0.3>
    Saffron=<rgba 1.5 1.3633 0.3203 0.6>
    SkyBlue=<rgba 0.625 1.5399 2 0.6>
    TitaniumWhite=<rgba 2 2 2 0.6>
    Use R1 Painted Overrides=true
    Use R2 Painted Overrides=true
   ) {module=VC_PaintedVariantsOverride001}
   &InheritSourceMovement(
    Applied Position Scale=&VectorFromFloat(Value=&FloatFromCurve(FloatCurve=DataInterface() {propertyValues {Curve {keys [<at 0.3 1 RCIM_Cubic>, {tangentMode=RCTM_User, time=1, value=0}]}}}))
    ApplyForce=false
    ApplyPosition=true
    ApplyPositionFalloff=false
   )
   &ParticleState()
   &SolveForcesAndVelocity(Clamp Velocity=false, Limit Acceleration=false)
   &VC_UpdateRibbon_v1(Enable Skeletal Position=true)
   &ScaleRibbonWidth(Ribbon Width Scale=&FloatFromCurve(FloatCurve=DataInterface() {propertyValues {Curve {
        defaultValue=1
        keys [
         <at 0 0.8 RCIM_Cubic>
         {interpMode=RCIM_Cubic, tangentMode=RCTM_User, time=0.2, value=1}
         {interpMode=RCIM_Cubic, tangentMode=RCTM_User, time=1, value=0.3}
        ]
       }}}))
   /SetVariables(
    Particles.RibbonColor1:LinearColor=&MultiplyLinearColor(
     Linear Color=$Particles.R1 Color
     Scale Factor=&MakeVector4FromVectorAndFloat(
      "Vector (RGB)"=&VectorFromFloat(Value=&MakeCustomFloatFromBool(Bool=$User.bIsPainted, True Float=3, False Float=5))
      "Float (Alpha)"=&FloatFromCurve(FloatCurve=DataInterface() {propertyValues {Curve {keys [<at 0 0>,<at 0.1 1>]}}})
     )
    )
    Particles.RibbonColor2:LinearColor=&MultiplyLinearColor(
     Linear Color=$Particles.R2 Color
     Scale Factor=&MakeVector4FromVectorAndFloat(
      "Vector (RGB)"=&VectorFromFloat(Value=5)
      "Float (Alpha)"=&FloatFromCurve(
       FloatCurve=DataInterface() {propertyValues {Curve {keys [<at 0 0>, {arriveTangent=-13.8363, leaveTangentWeight=0.6011, time=0.3, value=0.5}]}}}
       Scale Curve=0.2
      )
     )
    )
    Particles.RibbonScale1:NiagaraFloat=&Multiply_Float(
     A=45
     B=&FloatFromCurve(FloatCurve=DataInterface() {propertyValues {Curve {keys [<at 0 0.8>, {arriveTangent=-11.2102, leaveTangent=-11.2102, time=0.3, value=0.3}]}}})
    )
   ) {isSetParameters=true}
  ]
  rendererClasses [&NiagaraRibbonRendererProperties]
  renderers [
   {rendererClass=&NiagaraRibbonRendererProperties}
   {rendererClass=&NiagaraRibbonRendererProperties, rendererIndex=1}
   {rendererClass=&NiagaraRibbonRendererProperties, rendererIndex=2}
  ]
 }
}
summary {
 systemName=NS_DrifTrail_DeepSpace_MP
 emitters [
  {emitterName=NE_VC_DriftTrail_WheelWrap_Parent001, rendererClasses [&NiagaraMeshRendererProperties]}
  {
   emitterName=NE_VC_DriftTrail_Sprite_Parent
   rendererClasses [&NiagaraMeshRendererProperties, &NiagaraSpriteRendererProperties]
  }
  {emitterName=Ribbons002, rendererClasses [&NiagaraRibbonRendererProperties]}
 ]
 userVariables [
  User.bIsFrontendPreview {type=NiagaraBool}
  User.bIsPainted {type=NiagaraBool}
  User.ColorVariantID {type=NiagaraInt32}
  User.offset {type=Vector3f, description="offset from wheels(set through cosmetic template)", defaultValue {struct=Vector3f, value=<0 0 0>}}
  User.PaintedColor {type=LinearColor, description="", defaultValue {struct=LinearColor, value=<rgba 1 1 1 1>}}
  User.PaintedScalar {type=NiagaraFloat, defaultValue=0, description=""}
  User.VehicleFlags {type=NiagaraInt32, defaultValue=806289408}
  User.Wheel_Radius_Array {
   type=NiagaraDataInterfaceArrayFloat
   defaultValue {
    struct=NiagaraExt_VariableValue_DataInterface
    value {
     dataInterface=&self:NiagaraDataInterfaceArrayFloat_1
     dataInterfaceClass=&NiagaraDataInterfaceArrayFloat
    }
   }
  }
  User.Wheel_Width_Array {
   type=NiagaraDataInterfaceArrayFloat
   defaultValue {
    struct=NiagaraExt_VariableValue_DataInterface
    value {
     dataInterface=&self:NiagaraDataInterfaceArrayFloat_0
     dataInterfaceClass=&NiagaraDataInterfaceArrayFloat
    }
   }
  }
  User.WheelTouchingArray {
   type=NiagaraDataInterfaceArrayFloat
   defaultValue {
    struct=NiagaraExt_VariableValue_DataInterface
    value {
     dataInterface=&self:NiagaraDataInterfaceArrayFloat_2
     dataInterfaceClass=&NiagaraDataInterfaceArrayFloat
    }
   }
  }
 ]
}
